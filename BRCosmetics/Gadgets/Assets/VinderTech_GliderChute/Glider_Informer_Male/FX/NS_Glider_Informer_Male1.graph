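# Niagara effect for the Informer Male glider, with two short ribbon anim trails and blue-tinted animated pattern sprites.
# Two near-identical sprite emitters spawn only when fully deployed, each covering two skeletal sockets via spawn groups.
# Sprites re-sample their socket every update and drift along its rotated local axis, playing random SubUV frames.
# A system FN_Integrator value scales sprite spawn rate and all emitters' alpha, with front-end overrides for alpha and trail velocity.
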NiagaraSystem /BRCosmetics/Gadgets/Assets/VinderTech_GliderChute/Glider_Informer_Male/FX/NS_Glider_Informer_Male1
dependencies {
 usedDataInterfaces [&NiagaraDataInterfaceColorCurve, &NiagaraDataInterfaceCurve, &NiagaraDataInterfaceVector2DCurve]
 usedDynamicInputs [&Boolean_AndOperation, &SetBoolByFloatComparison]
 usedModules [&SystemState, &FN_Integrator]
}
emitters {
 FN_AnimTrail {
  emitterSpawnScript [&FN_RibbonIDOffsetUtility()]
  emitterUpdateScript [
   &EmitterState()
   &FN_SpawnAnimTrail(Number of Anim Trails=2)
  ]
  particleSpawnScript [
   &FN_InitializeAnimTrailBoneIndecies(First Bone or Socket Index=4, Second Bone or Socket Index=5)
   &FN_InitializeAnimTrailBoneIndecies(AnimTrail Selection=1, First Bone or Socket Index=6, Second Bone or Socket Index=7) {module=FN_InitializeAnimTrailBoneIndecies001}
   &FN_InitializeAnimTrail(AnimTrailWidth=2, Color=<rgba 0.8333 0.1799 0.1467 1>, Lifetime=0.1, Position Offset=<10 0 0>)
   &AddVelocity(Velocity=&FN_FrontEndVector3Modifier(Gameplay=<0 200 0>, Front End=<-400 0 0>))
  ]
  particleUpdateScript [
   &SolveForcesAndVelocity(Clamp Velocity=false, Limit Acceleration=false)
   &FN_UpdateAnimTrail(Locked Particle RibbonWidth=1, SkeletalMesh=$User.SkeletalMesh) {module=FN_UpdateAnimTrail001}
   /SetVariables(Particles.RibbonU0Override=$Particles.NormalizedAge) {isSetParameters=true}
   &ScaleColor(
    Scale Alpha=&FloatFromCurve(
     FloatCurve=DataInterface() {propertyValues {Curve {keys [
         {interpMode=RCIM_Cubic, tangentMode=RCTM_User, time=0, value=0}
         {interpMode=RCIM_Cubic, tangentMode=RCTM_User, time=0.5, value=1}
         {tangentMode=RCTM_User, time=1, value=1}
        ]}}}
     Scale Curve=&FN_FrontEndFloatModifier(Gameplay=$System.FN_Integrator.Value, Front End=1)
    )
    ScaleA=true
    ScaleRGB=true
   )
  ]
  rendererClasses [&NiagaraRibbonRendererProperties]
  renderers [{rendererClass=&NiagaraRibbonRendererProperties}]
 }
 FN_InformerPatterns_L {
  emitterSpawnScript [&FN_PreserveSkydiveUserVariables() {module=PreserveUserVariables}]
  emitterUpdateScript [
   &EmitterState()
   &SpawnRate(
    Spawn Probability=&MakeCustomFloatFromBool(Bool=$User.bIsFullyDeployed, True Float=1, False Float=0)
    SpawnRate=&Multiply_Float(A=40, B=&Lerp_Float(A=0.5, B=1, Alpha=$System.FN_Integrator.Value))
    Use Spawn Probability=true
   )
  ]
  particleSpawnScript [
   /SetVariables(Particles.RandomFrame:NiagaraFloat=&UniformRangedFloat(Minimum=0, Maximum=6)) {isSetParameters=true, module=SetVariables_#1}
   &FN_SetSpawnGroup_v2(Number of Spawn Groups=4)
   /Niagara/Modules/Spawn/Initialization/InitializeParticle(
    Color=<rgba 0.3288 0.5003 0.6979 1>
    Lifetime=0.9
    Module.Write Color=true
    Module.Write Lifetime=true
    Module.Write Mass=true
    Module.Write Material Random=true
    Module.Write Position=false
    Module.Write Scale=false
    Module.Write SpriteRotation=true
    Module.Write SpriteSize=true
    Position=&SimulationPosition()
    Sprite Rotation=90
    Sprite Size=<20 20>
   )
   /SetVariables(Particles.CustomSpriteFacing:Vector3f=<0 0 1>) {isSetParameters=true, module=SetVariables_#2}
   &CameraOffset(Camera Offset Amount=10)
  ]
  particleUpdateScript [
   /SetVariables(Particles.ColorShadow:LinearColor=&MultiplyLinearColor(
     Linear Color=&ColorFromCurve(ColorCurve=DataInterface() {propertyValues {
        AlphaCurve {keys [<at 0 0 RCIM_Cubic>,<at 0.25 1>,<at 0.5 0.5>,<at 1 0 RCIM_Cubic>]}
        BlueCurve {keys [<at 0.4 1.5>,<at 0.6783 0.6 RCIM_Cubic>]}
        GreenCurve {keys [<at 0.4 0.1693>,<at 0.6783 0.0677 RCIM_Cubic>]}
        RedCurve {keys [<at 0.4 0>,<at 0.6783 0 RCIM_Cubic>]}
       }})
     Scale Factor=&MakeVector4(X=1, Y=1, Z=1, W=$User.LocalAlpha)
    )) {isSetParameters=true, module=SetVariables_#3}
   &ParticleState()
   &SampleSkeletalMeshSkeleton(
    Mesh:NiagaraDataInterfaceSkeletalMesh=$User.SkeletalMesh
    Spawn Mode:ENiagaraSkeletalSpawning=::NewEnumerator1
    Specific Socket Index=HlslExpression() {hlslExpression=Particles.SpawnGroup/2}
   )
   &ScaleColor(
    Scale Alpha=&FloatFromCurve(
     FloatCurve=DataInterface() {propertyValues {Curve {keys [<at 0 0>, {interpMode=RCIM_Cubic, leaveTangent=-0.0074, tangentMode=RCTM_Break, time=0.5, value=1}, <at 1 0>]}}}
     Scale Curve=&FN_FrontEndFloatModifier(Gameplay=$System.FN_Integrator.Value, Front End=1)
    )
    Scale RGB=&MakeVectorFromLinearColorRGB(Color=&ColorFromCurve(ColorCurve=DataInterface() {propertyValues {
        AlphaCurve {keys [<at -0.0023 1 RCIM_Cubic>]}
        BlueCurve {keys [<at 0.05 1>,<at 0.4 2 RCIM_Cubic>,<at 0.65 1>,<at 0.85 0.5>]}
        GreenCurve {keys [
          <at 0.05 0.7606>
          {arriveTangent=-0.1548, interpMode=RCIM_Cubic, leaveTangent=-0.1548, time=0.4, value=2}
          <at 0.65 0.6677>
          <at 0.85 0.2323>
         ]}
        RedCurve {keys [<at 0.05 0.5677>, {arriveTangent=-0.2795, interpMode=RCIM_Cubic, leaveTangent=-0.2795, time=0.4, value=2}, <at 0.65 0.4>, <at 0.85 0>]}
       }}))
    ScaleA=true
    ScaleRGB=true
   )
   &ScaleSpriteSize(Scale Factor=&Vector2DFromCurve(
     Vector2Curve=DataInterface() {propertyValues {XCurve {keys [<at 0.25 0.8>,<at 1 1>]}, YCurve {keys [<at 0.25 0.8>,<at 1 1.5>]}}}
     ScaleCurve=<1 1>
    ))
   /Niagara/Modules/Update/SubUV/SubUVAnimation(Number Of Frames=6, SubUV Lookup Index=&Add_Float(A=$Particles.RandomFrame, B=$Particles.NormalizedAge))
   /SetVariables(Particles.OriginPosition:Vector3f=&Add_Vector(
     A=&RotateVectorByQuaternion(
      Vector To Rotate=&MakeVector(
       X=1
       Y=0
       Z=&self:SpawnGroupOffsetZ(Scale:NiagaraFloat=12)
      )
      Quaternion=$Particles.SampleSkeletalMeshSkeleton.SampledRotation
     )
     B=&Add_Vector(A=$Particles.SampleSkeletalMeshSkeleton.SampledPosition, B=<0 0 -8>)
    )) {isSetParameters=true, module=SetVariables_#4}
   /SetVariables(Particles.Position=&Lerp_Vector(
     A=$Particles.OriginPosition
     B=&Add_Vector(
      A=$Particles.OriginPosition
      B=&RotateVectorByQuaternion(Vector To Rotate=<0 150 5>, Quaternion=$Particles.SampleSkeletalMeshSkeleton.SampledRotation)
     )
     Alpha=$Particles.NormalizedAge
    )) {isSetParameters=true, module=SetVariables_#5}
  ]
  rendererClasses [&NiagaraSpriteRendererProperties]
  renderers [
   {rendererClass=&NiagaraSpriteRendererProperties}
   {rendererClass=&NiagaraSpriteRendererProperties, rendererIndex=1}
  ]
 }
 FN_InformerPatterns_L001 {
  emitterSpawnScript [&FN_PreserveSkydiveUserVariables() {module=PreserveUserVariables}]
  emitterUpdateScript [
   &EmitterState()
   &SpawnRate(
    Spawn Probability=&MakeCustomFloatFromBool(Bool=$User.bIsFullyDeployed, True Float=1, False Float=0)
    SpawnRate=&Multiply_Float(A=40, B=&Lerp_Float(A=0.5, B=1, Alpha=$System.FN_Integrator.Value))
    Use Spawn Probability=true
   )
  ]
  particleSpawnScript [
   /SetVariables(Particles.RandomFrame:NiagaraFloat=&UniformRangedFloat(Minimum=0, Maximum=6)) {isSetParameters=true, module=SetVariables_#1}
   &FN_SetSpawnGroup_v2(Number of Spawn Groups=4)
   &InitializeParticle(
    Color=<rgba 0.3288 0.5003 0.6979 1>
    Lifetime=0.9
    Module.Write Color=true
    Module.Write Lifetime=true
    Module.Write Mass=true
    Module.Write Material Random=true
    Module.Write Position=false
    Module.Write Scale=false
    Module.Write SpriteRotation=true
    Module.Write SpriteSize=true
    Position=&SimulationPosition()
    Sprite Rotation=90
    Sprite Size=<20 20>
   )
   /SetVariables(Particles.CustomSpriteFacing:Vector3f=<0 0 1>) {isSetParameters=true, module=SetVariables_#2}
   &CameraOffset(Camera Offset Amount=10)
  ]
  particleUpdateScript [
   /SetVariables(Particles.ColorShadow:LinearColor=&MultiplyLinearColor(
     Linear Color=&ColorFromCurve(ColorCurve=DataInterface() {propertyValues {
        AlphaCurve {keys [<at 0 0 RCIM_Cubic>,<at 0.25 1>,<at 0.5 0.5>,<at 1 0 RCIM_Cubic>]}
        BlueCurve {keys [<at 0.4 1.5>,<at 0.6783 0.6 RCIM_Cubic>]}
        GreenCurve {keys [<at 0.4 0.1693>,<at 0.6783 0.0677 RCIM_Cubic>]}
        RedCurve {keys [<at 0.4 0>,<at 0.6783 0 RCIM_Cubic>]}
       }})
     Scale Factor=&MakeVector4(X=1, Y=1, Z=1, W=$User.LocalAlpha)
    )) {isSetParameters=true, module=SetVariables_#3}
   &ParticleState()
   &SampleSkeletalMeshSkeleton(
    Mesh:NiagaraDataInterfaceSkeletalMesh=$User.SkeletalMesh
    Spawn Mode:ENiagaraSkeletalSpawning=::NewEnumerator1
    Specific Socket Index=HlslExpression() {hlslExpression=Particles.SpawnGroup/2+2}
   )
   &ScaleColor(
    Scale Alpha=&FloatFromCurve(
     FloatCurve=DataInterface() {propertyValues {Curve {keys [<at 0 0>, {interpMode=RCIM_Cubic, leaveTangent=-0.0074, tangentMode=RCTM_Break, time=0.5, value=1}, <at 1 0>]}}}
     Scale Curve=&FN_FrontEndFloatModifier(Gameplay=$System.FN_Integrator.Value, Front End=1)
    )
    Scale RGB=&MakeVectorFromLinearColorRGB(Color=&ColorFromCurve(ColorCurve=DataInterface() {propertyValues {
        AlphaCurve {keys [<at -0.0023 1 RCIM_Cubic>]}
        BlueCurve {keys [<at 0.05 1>,<at 0.4 2 RCIM_Cubic>,<at 0.65 1>,<at 0.85 0.5>]}
        GreenCurve {keys [
          <at 0.05 0.7606>
          {arriveTangent=-0.1548, interpMode=RCIM_Cubic, leaveTangent=-0.1548, time=0.4, value=2}
          <at 0.65 0.6677>
          <at 0.85 0.2323>
         ]}
        RedCurve {keys [<at 0.05 0.5677>, {arriveTangent=-0.2795, interpMode=RCIM_Cubic, leaveTangent=-0.2795, time=0.4, value=2}, <at 0.65 0.4>, <at 0.85 0>]}
       }}))
    ScaleA=true
    ScaleRGB=true
   )
   &ScaleSpriteSize(Scale Factor=&Vector2DFromCurve(
     Vector2Curve=DataInterface() {propertyValues {XCurve {keys [<at 0.25 0.8>,<at 1 1>]}, YCurve {keys [<at 0.25 0.8>,<at 1 1.5>]}}}
     ScaleCurve=<1 1>
    ))
   &SubUVAnimation(Number Of Frames=6, SubUV Lookup Index=&Add_Float(A=$Particles.RandomFrame, B=$Particles.NormalizedAge))
   /SetVariables(Particles.OriginPosition:Vector3f=&Add_Vector(
     A=&RotateVectorByQuaternion(
      Vector To Rotate=&MakeVector(
       X=1
       Y=0
       Z=&self:SpawnGroupOffsetZ(Scale:NiagaraFloat=10)
      )
      Quaternion=$Particles.SampleSkeletalMeshSkeleton.SampledRotation
     )
     B=&Add_Vector(A=$Particles.SampleSkeletalMeshSkeleton.SampledPosition, B=<0 0 -6>)
    )) {isSetParameters=true, module=SetVariables_#4}
   /SetVariables(Particles.Position=&Lerp_Vector(
     A=$Particles.OriginPosition
     B=&Add_Vector(
      A=$Particles.OriginPosition
      B=&RotateVectorByQuaternion(Vector To Rotate=<0 150 5>, Quaternion=$Particles.SampleSkeletalMeshSkeleton.SampledRotation)
     )
     Alpha=$Particles.NormalizedAge
    )) {isSetParameters=true, module=SetVariables_#5}
  ]
  rendererClasses [&NiagaraSpriteRendererProperties]
  renderers [
   {rendererClass=&NiagaraSpriteRendererProperties}
   {rendererClass=&NiagaraSpriteRendererProperties, rendererIndex=1}
  ]
 }
}
summary {
 systemName=NS_Glider_Informer_Male1
 emitters [
  {emitterName=FN_InformerPatterns_L, rendererClasses [&NiagaraSpriteRendererProperties]}
  {emitterName=FN_InformerPatterns_L001, rendererClasses [&NiagaraSpriteRendererProperties]}
  {emitterName=FN_AnimTrail, rendererClasses [&NiagaraRibbonRendererProperties]}
 ]
 userVariables [
  User.bIsFrontEnd {type=NiagaraBool}
  User.bIsFrontEndPreview {type=NiagaraBool}
  User.bIsFullyDeployed {type=NiagaraBool}
  User.ForwardDot {type=NiagaraFloat, defaultValue=0, description=""}
  User.LocalAlpha {type=NiagaraFloat, defaultValue=0, description=""}
  User.PatternSpawnMultiply {type=NiagaraFloat, defaultValue=0, description=""}
  User.PlayerSpeed {type=NiagaraFloat, defaultValue=0, description=""}
  User.RibbonIDOffset {type=NiagaraInt32}
  User.SkeletalMesh {
   type=NiagaraDataInterfaceSkeletalMesh
   defaultValue {
    struct=NiagaraExt_VariableValue_DataInterface
    value {dataInterface=&self:NiagaraDataInterfaceSkeletalMesh_0}
   }
  }
 ]
}
